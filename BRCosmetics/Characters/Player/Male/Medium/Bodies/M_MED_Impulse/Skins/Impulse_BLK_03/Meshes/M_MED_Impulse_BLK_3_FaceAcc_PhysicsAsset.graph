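# Physics asset for the face accessory mesh of the Impulse BLK 03 skin, a medium male player character cosmetic.
# It holds a single capsule body on the head bone, offset and rotated relative to that bone.
# With no constraints, it is a minimal one-body collision shape rather than a jointed ragdoll.
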PhysicsAsset /BRCosmetics/Characters/Player/Male/Medium/Bodies/M_MED_Impulse/Skins/Impulse_BLK_03/Meshes/M_MED_Impulse_BLK_3_FaceAcc_PhysicsAsset
bodies {head [{shapeType=Capsule, length=9.7557, radius=6.7335, center=<7.8821 -0.1897 -0.0544>, rotation=<rot 0 130.3776 -79.7182>}]}
constraints []
